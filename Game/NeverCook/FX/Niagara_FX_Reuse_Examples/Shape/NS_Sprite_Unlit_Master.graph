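# Niagara sprite effect from the FX reuse examples, named as an unlit sprite master: a brief bright cyan burst that grows and fades.
# A single Eye emitter burst-spawns one short-lived sprite with HDR color and material random.
# Float curves over its lifetime drive sprite growth and an overbright alpha fade.
# Camera-distance shrink parameters and User.VariantSwitch are declared, but no listed module reads them.
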NiagaraSystem /Game/NeverCook/FX/Niagara_FX_Reuse_Examples/Shape/NS_Sprite_Unlit_Master
dependencies {usedDataInterfaces [&NiagaraDataInterfaceCurve], usedDynamicInputs [], usedModules [&SystemState]}
emitters {Eye {
  emitterSpawnScript [
   /SetVariables(Emitter.Sprite Shrink Camera Distance=20) {isSetParameters=true}
   /SetVariables(Emitter.Sprite Shrink Distance Falloff Hardness=50) {isSetParameters=true}
  ]
  emitterUpdateScript [
   &EmitterState(Life Cycle Mode=::NewEnumerator1, Loop Behavior=::NewEnumerator1, Loop Duration=1, UseLoopDelay=false)
   &SpawnBurst_Instantaneous(Spawn Count=1, Use Spawn Probability=false)
  ]
  particleSpawnScript [&InitializeParticle(
    Color=<rgba 0.5 7.7833 10 1>
    Lifetime=0.3
    Sprite Size Mode=::NewEnumerator3
    Uniform Sprite Size=200
    UsePositionOffset=false
    Write Material Random=true
   ) {module=InitializeParticle001}]
  particleUpdateScript [
   &ParticleState()
   &ScaleSpriteSize(Scale Factor=&Vector2DFromFloat(Value=&FloatFromCurve(
      FloatCurve=DataInterface() {propertyValues {Curve {keys [{arriveTangent=3.9086, interpMode=RCIM_Cubic, leaveTangent=3.9086, tangentMode=RCTM_User, time=0, value=0}, <at 1 1.4 RCIM_Cubic>]}}}
      Scale Curve=0.6
     )))
   &ScaleColor(
    Scale Alpha=&FloatFromCurve(
     FloatCurve=DataInterface() {propertyValues {Curve {keys [
         {arriveTangent=-13.9062, interpMode=RCIM_Cubic, leaveTangent=-13.9062, tangentMode=RCTM_User, time=0, value=7}
         {arriveTangent=-7, interpMode=RCIM_Cubic, leaveTangent=-7, time=0.2488, value=3}
         <at 1 0 RCIM_Cubic>
        ]}}}
     Scale Curve=2
    )
    ScaleA=true
    ScaleRGB=true
   )
   &SolveForcesAndVelocity(Clamp Velocity=false, Limit Acceleration=false)
  ]
  rendererClasses [&NiagaraSpriteRendererProperties]
  renderers [{rendererClass=&NiagaraSpriteRendererProperties}]
 }}
summary {
 systemName=NS_Sprite_Unlit_Master
 emitters [{emitterName=Eye, rendererClasses [&NiagaraSpriteRendererProperties]}]
 userVariables [User.VariantSwitch {type=NiagaraFloat, defaultValue=0, description=""}]
}
